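# Niagara idle glow for the HybridBoot pickaxe, tracing a bow-like outline of string, limbs and tips.
# Burst spawn groups place persistent particles on hardcoded point arrays that follow the weapon through a socket reader.
# GlowMain aligns sprites along neighboring points, while Glow_Low uses fewer points with ribbons and sprites.
# A color curve scrolls along the shape, with fixed-color end glows; alpha varies by time of day and front-end versus gameplay.
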NiagaraSystem /BRCosmetics/Weapons/FORT_Melee/Pickaxe_HybridBoot/FX/NS_Pickaxe_HybridBoot_IdleGlow
dependencies {
 usedDataInterfaces [
  &NiagaraDataInterfaceArrayFloat3
  &NiagaraDataInterfaceColorCurve
  &NiagaraDataInterfaceCurve
  &NiagaraDataInterfaceSocketReader
 ]
 usedDynamicInputs [&FNTimeOfDayScalar]
 usedModules [
  &self:SystemSpawnScript.SystemScriptSource.SystemScriptGraph.NiagaraNodeAssignment_0.SetVariables_#1
  &SystemState
  &self:SystemSpawnScript.SystemScriptSource.SystemScriptGraph.NiagaraNodeAssignment_1.SetVariables_#2
 ]
}
emitters {
 GlowMain {
  emitterSpawnScript [/SetVariables(Emitter.BowCurveSpawnCount=8, Emitter.BowStringSpawnCount=15, Emitter.BowTipCurveSpawnCount=5) {isSetParameters=true, module=SetVariables_#3}]
  emitterUpdateScript [
   &EmitterState(
    Inactive Response=::NewEnumerator1
    Life Cycle Mode=::NewEnumerator1
    Loop Behavior=::NewEnumerator1
    Loop Duration Mode=::NewEnumerator1
    UseLoopDelay=false
   )
   &SpawnBurst_Instantaneous(Spawn Count=$Emitter.BowStringSpawnCount, Use Spawn Probability=false, UseLoopCountLimit=false)
   &SpawnBurst_Instantaneous(Spawn Count=$Emitter.BowCurveSpawnCount, Spawn Group=1, Use Spawn Probability=false, UseLoopCountLimit=false) {module=SpawnBurst_Instantaneous001}
   &SpawnBurst_Instantaneous(Spawn Count=$Emitter.BowCurveSpawnCount, Spawn Group=2, Use Spawn Probability=false, UseLoopCountLimit=false) {module=SpawnBurst_Instantaneous002}
   &SpawnBurst_Instantaneous(Spawn Count=2, Spawn Group=3, Use Spawn Probability=false, UseLoopCountLimit=false) {module=SpawnBurst_Instantaneous003}
   &SpawnBurst_Instantaneous(Spawn Count=$Emitter.BowTipCurveSpawnCount, Spawn Group=4, Use Spawn Probability=false, UseLoopCountLimit=false) {module=SpawnBurst_Instantaneous004}
   &SpawnBurst_Instantaneous(Spawn Count=$Emitter.BowTipCurveSpawnCount, Spawn Group=5, Use Spawn Probability=false, UseLoopCountLimit=false) {module=SpawnBurst_Instantaneous005}
  ]
  particleSpawnScript [
   /SetVariables(
    Particles.Array01=DataInterface() {propertyValues {
      GpuSyncMode=SyncCpuToGpu
      MaxElements=0
      FloatData [
       <1.91 -5.372 15.94>
       <1.836 -5.074 20.363>
       <1.304 -3.669 24.535>
       <0.532 -1.63 28.396>
       <-0.389 0.807 31.985>
       <-1.362 3.384 35.459>
       <-2.341 5.983 38.916>
       <-3.302 8.448 42.474>
      ]
     }}
    Particles.Array02=DataInterface() {propertyValues {
      GpuSyncMode=SyncCpuToGpu
      MaxElements=0
      FloatData [
       <-3.328 8.475 -44.054>
       <-2.363 6.004 -40.51>
       <-1.383 3.443 -37.036>
       <-0.399 0.835 -33.597>
       <0.519 -1.595 -30.013>
       <1.291 -3.64 -26.164>
       <1.819 -5.038 -21.997>
       <1.955 -5.397 -17.587>
      ]
     }}
    Particles.Array03=DataInterface() {propertyValues {
      GpuSyncMode=SyncCpuToGpu
      MaxElements=0
      FloatData [
       <-5.025 13.058 60.018>
       <-5.199 13.619 51.444>
       <-5.227 13.731 42.85>
       <-5.238 13.759 34.255>
       <-5.258 13.803 25.661>
       <-5.273 13.853 17.066>
       <-5.293 13.902 8.472>
       <-5.299 13.918 -0.123>
       <-5.292 13.902 -8.717>
       <-5.278 13.86 -17.312>
       <-5.262 13.817 -25.906>
       <-5.247 13.775 -34.501>
       <-5.227 13.726 -43.095>
       <-5.196 13.644 -51.689>
       <-5.045 13.182 -60.27>
      ]
     }}
    Particles.Array04=DataInterface() {propertyValues {
      GpuSyncMode=SyncCpuToGpu
      MaxElements=0
      FloatData [<-3.852 9.787 67.677>, <-4.221 10.878 66.499>, <-4.558 11.808 65.182>, <-4.8 12.454 63.685>, <-4.976 12.842 62.094>]
     }}
    Particles.Array05=DataInterface() {propertyValues {
      GpuSyncMode=SyncCpuToGpu
      MaxElements=0
      FloatData [<-5.005 12.935 -62.864>, <-4.88 12.616 -64.674>, <-4.619 11.989 -66.388>, <-4.273 10.999 -67.903>, <-3.852 9.787 -69.225>]
     }}
   ) {isSetParameters=true, module=SetVariables_#4}
   /SetVariables(
    Particles.Point:Vector3f=&MergeHelper_VectorSwitch(
     0=&SelectVectorFromArray(Direct Array Index=&ReturnExecIndex(), Vector Selection Array=$Particles.Array03)
     1=&SelectVectorFromArray(Direct Array Index=&ReturnExecIndex(), Vector Selection Array=$Particles.Array01)
     2=&SelectVectorFromArray(Direct Array Index=&ReturnExecIndex(), Vector Selection Array=$Particles.Array02)
     3=&SelectVectorFromArray(
      Direct Array Index=&ReturnExecIndex()
      Vector Selection Array=DataInterface() {propertyValues {FloatData [<-3.982 10.206 71.252>, <-4.001 10.246 -72.162>]}}
     )
     4=&SelectVectorFromArray(Direct Array Index=&ReturnExecIndex(), Vector Selection Array=$Particles.Array04)
     5=&SelectVectorFromArray(Direct Array Index=&ReturnExecIndex(), Vector Selection Array=$Particles.Array05)
     Index=$Emitter.SpawnGroup
     Num=::NewEnumerator4
    )
    Particles.PrevPoint:Vector3f=&MergeHelper_VectorSwitch(
     0=&SelectVectorFromArray(
      Direct Array Index=&MakeIntFromFloat(Float=&self:ReturnPrevIndexCenter(
        CurrentIndex:NiagaraInt32=&ReturnExecIndex()
        ArrayLength:NiagaraInt32=$Emitter.BowStringSpawnCount
        CenterIndex:NiagaraFloat=8
       ))
      Vector Selection Array=$Particles.Array03
     )
     1=&SelectVectorFromArray(
      Direct Array Index=&MakeIntFromFloat(Float=&self:ReturnPrevIndexCenter(
        CurrentIndex:NiagaraInt32=&ReturnExecIndex()
        ArrayLength:NiagaraInt32=$Emitter.BowCurveSpawnCount
        CenterIndex:NiagaraFloat=4
       ))
      Vector Selection Array=$Particles.Array01
     )
     2=&SelectVectorFromArray(
      Direct Array Index=&MakeIntFromFloat(Float=&self:ReturnPrevIndexCenter(
        CurrentIndex:NiagaraInt32=&ReturnExecIndex()
        ArrayLength:NiagaraInt32=$Emitter.BowCurveSpawnCount
        CenterIndex:NiagaraFloat=4
       ))
      Vector Selection Array=$Particles.Array02
     )
     3=<0 0 0>
     4=&SelectVectorFromArray(
      Direct Array Index=&MakeIntFromFloat(Float=&self:ReturnPrevIndexCenter(
        CurrentIndex:NiagaraInt32=&ReturnExecIndex()
        ArrayLength:NiagaraInt32=$Emitter.BowTipCurveSpawnCount
        CenterIndex:NiagaraFloat=3
       ))
      Vector Selection Array=$Particles.Array04
     )
     5=&SelectVectorFromArray(
      Direct Array Index=&MakeIntFromFloat(Float=&self:ReturnPrevIndexCenter(
        CurrentIndex:NiagaraInt32=&ReturnExecIndex()
        ArrayLength:NiagaraInt32=$Emitter.BowTipCurveSpawnCount
        CenterIndex:NiagaraFloat=3
       ))
      Vector Selection Array=$Particles.Array05
     )
     Index=$Emitter.SpawnGroup
     Num=::NewEnumerator4
    )
   ) {isSetParameters=true, module=SetVariables_#5}
   &InitializeParticle(
    Sprite Rotation Mode=::NewEnumerator0
    Sprite Size=&MergeHelper_Vector2Switch(0=<5.1854 13.1868>, 1=<9 16.5548>, 2=<9 16.5548>, 3=<22.05 36.225>, 4=<8 7>, 5=<8 7>, Index=$Emitter.SpawnGroup, Num=::NewEnumerator4)
    Sprite Size Mode=::NewEnumerator4
    UsePositionOffset=false
    Write Material Random=true
   )
   /SetVariables(Particles.SpawnGroup=$Emitter.SpawnGroup) {isSetParameters=true, module=SetVariables_#6}
   /SetVariables(Particles.NormalizedIndex:NiagaraFloat=&MergeHelper_FloatSwitch(
     0=&Lerp_Float(A=0.2, B=0.8, Alpha=&ReturnNormalizedExecIndex())
     1=&Lerp_Float(A=0.43, B=0.3, Alpha=&ReturnNormalizedExecIndex())
     2=&Lerp_Float(A=0.73, B=0.59, Alpha=&ReturnNormalizedExecIndex())
     3=0
     4=&Lerp_Float(A=0, B=0.15, Alpha=&ReturnNormalizedExecIndex())
     5=&Lerp_Float(A=0.85, B=1, Alpha=&ReturnNormalizedExecIndex())
     Index=$Particles.SpawnGroup
     Num=::NewEnumerator4
    )) {isSetParameters=true}
  ]
  particleUpdateScript [
   &ParticleState(Kill Particles When Lifetime Has Elapsed=false, Let Infinitely Lived Particles Die When Emitter Deactivates=true)
   &SocketLocation(
    Orientation Sampling=::NewEnumerator1
    SamplePosOffset=true
    Sampled Position Offset=&AddVectorToPosition(
     Position=Linked() {linkedVariable {name=Particles.Point, type=Vector3f}}
     Vector=&MergeHelper_VectorSwitch(
      0=<5.2331 -13.7343 -15.6142>
      1=<5.2331 -13.7343 -15.6142>
      2=<5.2331 -13.7343 -15.6142>
      3=<5.2331 -13.7343 -15.6142>
      4=<5.2331 -13.7343 -15.6142>
      5=<5.2331 -13.7343 -15.6142>
      Index=$Particles.SpawnGroup
      Num=::NewEnumerator4
     )
    )
    Socket Reader=$System.Socket Reader
    Socket Sampling Mode=::NewEnumerator2
    VelocitySpeedLimit=false
   )
   &SpriteFacingAndAlignment(
    Sprite Alignment=&MaskVector(
     Mask Source=0
     Mask Parameter=?
     Mask=3
     Masked Vector=<0 0 1>
     Passthrough Vector=&RotateVectorByQuaternion(
      Vector To Rotate=&NormalizeVector(
       Vector To Normalize=&Subtract_Vector(A=$Particles.Point, B=$Particles.PrevPoint)
       Safe Normalize=true
       Safe Fallback Threshold=0
       Safe Fallback Vector=<0 0 0.001>
      )
      Quaternion=$Output.SocketLocation.SampledRotation
     )
    )
    WriteAlignment=true
    WriteFacing=false
   )
   &ScaleColor(
    Scale Alpha=&Multiply_Float(
     A=&FN_FrontEndFloatModifier(
      Gameplay=0.185
      Front End=&FloatFromCurve(FloatCurve=DataInterface() {propertyValues {Curve {keys [<at 0.2 0>,<at 0.7 1>]}}}, CurveIndex=$System.Age, Scale Curve=0.135)
     )
     B=$System.TODScalar
    )
    Scale RGB=&MaskVector(
     Mask Source=1
     Mask Parameter=$Particles.SpawnGroup
     Mask=3
     Masked Vector=&FN_FrontEndVector3Modifier(
      Gameplay=&MakeVectorFromLinearColorRGB(Color=<rgba 0.5395 0.552 0.3813 1>)
      Front End=&MakeVectorFromLinearColorRGB(Color=<rgba 0.3909 0.4 0.2763 1>)
     )
     Passthrough Vector=&MakeVectorFromLinearColorRGB(Color=&ColorFromCurve(
       ColorCurve=DataInterface() {propertyValues {
         AlphaCurve {keys []}
         BlueCurve {keys [<at 0 1>,<at 0.0267 1>,<at 0.1762 1>,<at 0.2153 1>,<at 0.2417 1>,<at 0.285 1>,<at 0.3625 1>,<at 0.3946 1>,<at 0.4565 1>,<at 0.5395 1>,<at 0.5912 1>,<at 0.6172 1>,<at 0.7148 1>,<at 0.748 1>,<at 0.7793 1>,<at 0.8881 1>,<at 0.9453 1>]}
         GreenCurve {keys [<at 0 0.0123>,<at 0.0267 0.159>,<at 0.1762 0.159>,<at 0.2153 0.624>,<at 0.2417 0.624>,<at 0.285 0.159>,<at 0.3625 0.624>,<at 0.3946 0.624>,<at 0.4565 0.159>,<at 0.5395 0.159>,<at 0.5912 0.624>,<at 0.6172 0.624>,<at 0.7148 0.159>,<at 0.748 0.624>,<at 0.7793 0.159>,<at 0.8881 0.159>,<at 0.9453 0.0123>]}
         RedCurve {keys [<at 0 0.491>,<at 0.0267 0.2961>,<at 0.1762 0.2961>,<at 0.2153 0.2384>,<at 0.2417 0.2384>,<at 0.285 0.2961>,<at 0.3625 0.2384>,<at 0.3946 0.2384>,<at 0.4565 0.2961>,<at 0.5395 0.2961>,<at 0.5912 0.2384>,<at 0.6172 0.2384>,<at 0.7148 0.2961>,<at 0.748 0.2384>,<at 0.7793 0.2961>,<at 0.8881 0.2961>,<at 0.9453 0.491>]}
        }}
       CurveIndex=&Frac_Float(Float=&Add_Float(
         A=$Particles.NormalizedIndex
         B=&OneMinusFloat(Float=&Frac_Float(Float=&Multiply_Float(A=$Engine.Time, B=0.035)))
        ))
      ))
    )
    ScaleA=true
    ScaleRGB=true
   )
  ]
  rendererClasses [&NiagaraSpriteRendererProperties]
  renderers [{rendererClass=&NiagaraSpriteRendererProperties}]
 }
 Glow_Low {
  emitterSpawnScript [/SetVariables(Emitter.BowCurveSpawnCount=8, Emitter.BowStringSpawnCount=13) {isSetParameters=true, module=SetVariables_#3}]
  emitterUpdateScript [
   &EmitterState(
    Inactive Response=::NewEnumerator1
    Life Cycle Mode=::NewEnumerator1
    Loop Behavior=::NewEnumerator1
    Loop Duration Mode=::NewEnumerator1
    UseLoopDelay=false
   )
   &SpawnBurst_Instantaneous(Spawn Count=$Emitter.BowStringSpawnCount, Use Spawn Probability=false, UseLoopCountLimit=false) {module=SpawnBurst_Instantaneous003}
   &SpawnBurst_Instantaneous(Spawn Count=$Emitter.BowCurveSpawnCount, Spawn Group=1, Use Spawn Probability=false, UseLoopCountLimit=false) {module=SpawnBurst_Instantaneous001}
   &SpawnBurst_Instantaneous(Spawn Count=$Emitter.BowCurveSpawnCount, Spawn Group=2, Use Spawn Probability=false, UseLoopCountLimit=false) {module=SpawnBurst_Instantaneous002}
   &SpawnBurst_Instantaneous(Spawn Group=3, Use Spawn Probability=false, UseLoopCountLimit=false)
   &SpawnBurst_Instantaneous(Spawn Group=4, Use Spawn Probability=false, UseLoopCountLimit=false) {module=SpawnBurst_Instantaneous004}
  ]
  particleSpawnScript [
   &InitializeParticle(
    Ribbon Width=&MergeHelper_FloatSwitch(0=4, 1=8, 2=8, 3=?, Index=$Emitter.SpawnGroup, Num=::NewEnumerator1)
    Ribbon Width Mode=::NewEnumerator1
    Sprite Size=<22.05 36.225>
    Sprite Size Mode=::NewEnumerator4
    UsePositionOffset=false
    Write Material Random=true
   )
   /SetVariables(
    Particles.Array01=DataInterface() {propertyValues {
      GpuSyncMode=SyncCpuToGpu
      MaxElements=0
      FloatData [
       <-3.852 9.787 67.677>
       <-4.558 11.808 65.182>
       <-4.976 12.842 62.094>
       <-5.199 13.619 51.444>
       <-5.238 13.759 34.255>
       <-5.273 13.853 17.066>
       <-5.299 13.918 -0.123>
       <-5.278 13.86 -17.312>
       <-5.247 13.775 -34.501>
       <-5.196 13.644 -51.689>
       <-5.005 12.935 -62.864>
       <-4.619 11.989 -66.388>
       <-3.852 9.787 -69.225>
      ]
     }}
    Particles.Array02=DataInterface() {propertyValues {
      GpuSyncMode=SyncCpuToGpu
      MaxElements=0
      FloatData [<1.804 -5.02 14.363>, <1.886 -5.207 19.37>, <0.5 -1.544 28.531>, <-1.656 4.165 36.477>, <-3.752 9.648 44.593>]
     }}
    Particles.Array03=DataInterface() {propertyValues {
      GpuSyncMode=SyncCpuToGpu
      MaxElements=0
      FloatData [<-3.752 9.648 -46.142>, <-1.665 4.184 -38.012>, <0.501 -1.549 -30.089>, <1.872 -5.178 -20.918>, <1.804 -5.02 -15.912>]
     }}
   ) {isSetParameters=true, module=SetVariables_#4}
   /SetVariables(Particles.RibbonID=&self:MakeRibbonID(ID:NiagaraInt32=&MergeHelper_IntSwitch(0=0, 1=1, 2=2, 3=3, 4=4, 5=?, Index=$Emitter.SpawnGroup, Num=::NewEnumerator3))) {isSetParameters=true}
   /SetVariables(Particles.Point:Vector3f=&MergeHelper_VectorSwitch(
     0=&SelectVectorFromArray(Direct Array Index=&ReturnExecIndex(), Vector Selection Array=$Particles.Array01)
     1=&SelectVectorFromArray(Direct Array Index=&ReturnExecIndex(), Vector Selection Array=$Particles.Array02)
     2=&SelectVectorFromArray(Direct Array Index=&ReturnExecIndex(), Vector Selection Array=$Particles.Array03)
     3=<-3.982 10.206 71.252>
     4=<-4.001 10.246 -72.162>
     Index=$Emitter.SpawnGroup
     Num=::NewEnumerator3
    )) {isSetParameters=true, module=SetVariables_#5}
   /SetVariables(Particles.SpawnGroup=/Niagara/DynamicInputs/Helpers/MaskInt(
     Mask Source:NiagaraInt32=1
     Mask Parameter:NiagaraInt32=$Emitter.SpawnGroup
     Mask:NiagaraInt32=3
     Masked Int:NiagaraInt32=3
     Passthrough Int:NiagaraInt32=&MaskInt(
      Mask Source:NiagaraInt32=1
      Mask Parameter:NiagaraInt32=$Emitter.SpawnGroup
      Mask:NiagaraInt32=4
      Masked Int:NiagaraInt32=3
      Passthrough Int:NiagaraInt32=$Emitter.SpawnGroup
     )
    )) {isSetParameters=true, module=SetVariables_#6}
   /SetVariables(Particles.NormalizedIndex:NiagaraFloat=&MergeHelper_FloatSwitch(
     0=&ReturnNormalizedExecIndex()
     1=&Lerp_Float(A=0.43, B=0.3, Alpha=&ReturnNormalizedExecIndex())
     2=&Lerp_Float(A=0.73, B=0.59, Alpha=&ReturnNormalizedExecIndex())
     3=?
     Index=$Emitter.SpawnGroup
     Num=::NewEnumerator1
    )) {isSetParameters=true}
  ]
  particleUpdateScript [
   &ParticleState(Kill Particles When Lifetime Has Elapsed=false, Let Infinitely Lived Particles Die When Emitter Deactivates=true)
   &SocketLocation(
    Orientation Sampling=::NewEnumerator1
    SamplePosOffset=true
    Sampled Position Offset=&AddVectorToPosition(
     Position=Linked() {linkedVariable {name=Particles.Point, type=Vector3f}}
     Vector=&MergeHelper_VectorSwitch(
      0=<5.2331 -13.7343 -15.6142>
      1=<5.2331 -13.7343 -15.6142>
      2=<5.2331 -13.7343 -15.6142>
      3=<5.2331 -13.7343 -15.6142>
      Index=$Particles.SpawnGroup
      Num=::NewEnumerator2
     )
    )
    Socket Reader=$System.Socket Reader
    Socket Sampling Mode=::NewEnumerator2
    VelocitySpeedLimit=false
   )
   &SpriteFacingAndAlignment(
    Sprite Alignment=&RotateVectorByQuaternion(Vector To Rotate=<0 0 1>, Quaternion=$Output.SocketLocation.SampledRotation)
    WriteAlignment=true
    WriteFacing=false
   )
   &ScaleColor(
    Scale Alpha=&Multiply_Float(
     A=&FN_FrontEndFloatModifier(
      Gameplay=0.3
      Front End=&FloatFromCurve(FloatCurve=DataInterface() {propertyValues {Curve {keys [<at 0.2 0>,<at 0.7 1>]}}}, CurveIndex=$System.Age, Scale Curve=0.35)
     )
     B=$System.TODScalar
    )
    Scale RGB=&MaskVector(
     Mask Source=1
     Mask Parameter=$Particles.SpawnGroup
     Mask=3
     Masked Vector=&FN_FrontEndVector3Modifier(
      Gameplay=&MakeVectorFromLinearColorRGB(Color=<rgba 0.1808 0.185 0.1278 1>)
      Front End=&MakeVectorFromLinearColorRGB(Color=<rgba 0.1613 0.165 0.114 1>)
     )
     Passthrough Vector=&MakeVectorFromLinearColorRGB(Color=&ColorFromCurve(
       ColorCurve=DataInterface() {propertyValues {
         AlphaCurve {keys []}
         BlueCurve {keys [<at 0 1>,<at 0.0267 1>,<at 0.1762 1>,<at 0.2153 1>,<at 0.2417 1>,<at 0.285 1>,<at 0.3625 1>,<at 0.3946 1>,<at 0.4565 1>,<at 0.5395 1>,<at 0.5912 1>,<at 0.6172 1>,<at 0.7148 1>,<at 0.748 1>,<at 0.7793 1>,<at 0.8881 1>,<at 0.9453 1>]}
         GreenCurve {keys [<at 0 0.0123>,<at 0.0267 0.159>,<at 0.1762 0.159>,<at 0.2153 0.624>,<at 0.2417 0.624>,<at 0.285 0.159>,<at 0.3625 0.624>,<at 0.3946 0.624>,<at 0.4565 0.159>,<at 0.5395 0.159>,<at 0.5912 0.624>,<at 0.6172 0.624>,<at 0.7148 0.159>,<at 0.748 0.624>,<at 0.7793 0.159>,<at 0.8881 0.159>,<at 0.9453 0.0123>]}
         RedCurve {keys [<at 0 0.491>,<at 0.0267 0.2961>,<at 0.1762 0.2961>,<at 0.2153 0.2384>,<at 0.2417 0.2384>,<at 0.285 0.2961>,<at 0.3625 0.2384>,<at 0.3946 0.2384>,<at 0.4565 0.2961>,<at 0.5395 0.2961>,<at 0.5912 0.2384>,<at 0.6172 0.2384>,<at 0.7148 0.2961>,<at 0.748 0.2384>,<at 0.7793 0.2961>,<at 0.8881 0.2961>,<at 0.9453 0.491>]}
        }}
       CurveIndex=&Frac_Float(Float=&Add_Float(
         A=$Particles.NormalizedIndex
         B=&OneMinusFloat(Float=&Frac_Float(Float=&Multiply_Float(A=$Engine.Time, B=0.035)))
        ))
      ))
    )
    ScaleA=true
    ScaleRGB=true
   )
  ]
  rendererClasses [&NiagaraRibbonRendererProperties, &NiagaraSpriteRendererProperties]
  renderers [
   {rendererClass=&NiagaraRibbonRendererProperties}
   {rendererClass=&NiagaraSpriteRendererProperties, rendererIndex=1}
  ]
 }
}
summary {
 systemName=NS_Pickaxe_HybridBoot_IdleGlow
 emitters [
  {emitterName=GlowMain, rendererClasses [&NiagaraSpriteRendererProperties]}
  {emitterName=Glow_Low, rendererClasses [&NiagaraRibbonRendererProperties, &NiagaraSpriteRendererProperties]}
 ]
 userVariables [
  User.bIsFrontEnd {type=NiagaraBool}
  User.bIsFrontEndPreview {type=NiagaraBool}
 ]
}
