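# Niagara swing trail for the HybridBoot pickaxe, built from two ribbon emitters that follow bone or socket pairs through shared anim-trail functions.
# MainTrail cycles a blue-to-purple gradient along its length by offsetting particle age with engine time, under a brightness falloff.
# TrailTip adds a pale, narrowing streak on another socket pair.
# Both scale alpha by time of day and SwingGeneralFade, and animate material parameters over emitter age.
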NiagaraSystem /BRCosmetics/Weapons/FORT_Melee/Pickaxe_HybridBoot/FX/NS_Pickaxe_HybridBoot_Trail
dependencies {
 usedDataInterfaces [&NiagaraDataInterfaceColorCurve, &NiagaraDataInterfaceCurve, &NiagaraDataInterfaceSocketReader]
 usedDynamicInputs [&MergeHelper_FloatSwitch]
 usedModules [
  &self:SystemSpawnScript.SystemScriptSource.SystemScriptGraph.NiagaraNodeAssignment_1.SetVariables_#1
  &self:SystemSpawnScript.SystemScriptSource.SystemScriptGraph.NiagaraNodeAssignment_3.SetVariables_#2
  &self:SystemSpawnScript.SystemScriptSource.SystemScriptGraph.NiagaraNodeAssignment_4.SetVariables_#3
  &SystemState
  &self:SystemSpawnScript.SystemScriptSource.SystemScriptGraph.NiagaraNodeAssignment_2.SetVariables_#4
 ]
}
emitters {
 MainTrail {
  emitterSpawnScript [&FN_RibbonIDOffsetUtility()]
  emitterUpdateScript [
   &EmitterState()
   &FN_SpawnAnimTrail()
  ]
  particleSpawnScript [
   &FN_InitializeAnimTrailBoneIndecies()
   &FN_InitializeAnimTrail(AnimTrailWidth=1.25, Color=<rgba 10 10 10 1>, Lifetime=0.22)
  ]
  particleUpdateScript [
   &FN_UpdateAnimTrail() {module=FN_UpdateAnimTrail001}
   &ScaleColor(
    Scale Alpha=&FNTimeOfDayScalar(Night Scale=0.35, Morning Scale=0.45, Day Scale=0.5, Evening Scale=0.4)
    Scale RGB=&Multiply_Vector(
     A=&MakeVectorFromLinearColorRGB(Color=&ColorFromCurve(ColorCurve=DataInterface() {propertyValues {
         AlphaCurve {keys []}
         BlueCurve {keys [<at 0 4 RCIM_Cubic>,<at 0.1577 3>,<at 0.6244 1.5 RCIM_Cubic>]}
         GreenCurve {keys [<at 0 4 RCIM_Cubic>,<at 0.1577 3>,<at 0.6244 1.5 RCIM_Cubic>]}
         RedCurve {keys [<at 0 4 RCIM_Cubic>,<at 0.1577 3>,<at 0.6244 1.5 RCIM_Cubic>]}
        }}))
     B=&MakeVectorFromLinearColorRGB(Color=&ColorFromCurve(
       ColorCurve=DataInterface() {propertyValues {
         AlphaCurve {keys []}
         BlueCurve {keys [<at 0 0.7>, {arriveTangent=0.4, interpMode=RCIM_Cubic, leaveTangent=0.4, time=0.25, value=1}, <at 0.75 1>, <at 1 0.7 RCIM_Cubic>]}
         GreenCurve {keys [
           <at 0 0.4273>
           {arriveTangent=-0.4365, interpMode=RCIM_Cubic, leaveTangent=-0.4365, time=0.25, value=0.065}
           <at 0.75 0.1>
           <at 1 0.4273 RCIM_Cubic>
          ]}
         RedCurve {keys [
           <at 0 0.1511>
           {arriveTangent=0.1276, interpMode=RCIM_Cubic, leaveTangent=0.1276, time=0.25, value=0.2987}
           <at 0.75 0.2468>
           <at 1 0.1511 RCIM_Cubic>
          ]}
        }}
       CurveIndex=&Frac_Float(Float=&Add_Float(
         A=$Particles.NormalizedAge
         B=&OneMinusFloat(Float=&Frac_Float(Float=&Multiply_Float(A=$Engine.Time, B=0.2)))
        ))
      ))
    )
    ScaleA=true
    ScaleRGB=true
   )
   &ScaleColor(Scale Alpha=$System.SwingGeneralFade, ScaleA=true, ScaleRGB=true) {module=ScaleColor001}
   &DynamicMaterialParameters(
    Index 0 Param 3=0.0265
    Index 0 Param 4=1
    Index 1 Param 3=&FloatFromCurve(FloatCurve=DataInterface() {propertyValues {Curve {keys [<at 0 0>,<at 1 1>]}}}, CurveIndex=$Emitter.Age, Scale Curve=-2)
    Index 1 Param 4=0.165
    Index 2 Param 1=20
    Index 2 Param 2=2
    Index 2 Param 4=0.6
    Module.Index0Float0Write=true
    Module.Index0Float1Write=true
    Module.Index0Float2Write=true
    Module.Index0Float3Write=true
    Module.Index1Float0Write=true
    Module.Index1Float1Write=true
    Module.Index1Float2Write=true
    Module.Index1Float3Write=true
    Module.Index2Float0Write=true
    Module.Index2Float1Write=true
    Module.Index2Float2Write=false
    Module.Index2Float3Write=false
    Write Parameter Index 1=::NewEnumerator1
    Write Parameter Index 2=::NewEnumerator1
   )
  ]
  rendererClasses [&NiagaraRibbonRendererProperties]
  renderers [{rendererClass=&NiagaraRibbonRendererProperties}]
 }
 TrailTip {
  emitterSpawnScript [&FN_RibbonIDOffsetUtility()]
  emitterUpdateScript [
   &EmitterState()
   &FN_SpawnAnimTrail()
  ]
  particleSpawnScript [
   &FN_InitializeAnimTrailBoneIndecies(First Bone or Socket Index=1, Second Bone or Socket Index=2)
   &FN_InitializeAnimTrail(Color=<rgba 0.9794 1 0.691 1>, Lifetime=0.22)
  ]
  particleUpdateScript [
   &FN_UpdateAnimTrail() {module=FN_UpdateAnimTrail001}
   &ScaleColor(
    Scale Alpha=&FloatFromCurve(
     FloatCurve=DataInterface() {propertyValues {Curve {keys [
         {interpMode=RCIM_Cubic, leaveTangent=-5.3456, tangentMode=RCTM_Break, time=0, value=1.5}
         <at 0.6689 0.0233>
         {arriveTangent=-0.2139, interpMode=RCIM_Cubic, tangentMode=RCTM_Break, time=1, value=0}
        ]}}}
     Scale Curve=&FNTimeOfDayScalar(Night Scale=0.3375, Morning Scale=0.4125, Day Scale=0.45, Evening Scale=0.375)
    )
    Scale RGB=&VectorFromFloat(Value=&FloatFromCurve(FloatCurve=DataInterface() {propertyValues {Curve {keys [<at 0 4>,<at 0.3 1>,<at 1 0>]}}}))
    ScaleA=true
    ScaleRGB=true
   )
   &ScaleRibbonWidth(Ribbon Width Scale=&FloatFromCurve(FloatCurve=DataInterface() {propertyValues {Curve {keys [<at 0 1>,<at 0.2 0.35>,<at 1 0.2>]}}}))
   &ScaleColor(Scale Alpha=$System.SwingGeneralFade, ScaleA=true, ScaleRGB=true) {module=ScaleColor001}
   &DynamicMaterialParameters(
    Index 1 Param 3=&FloatFromCurve(FloatCurve=DataInterface() {propertyValues {Curve {keys [<at 0 0>,<at 1 1>]}}}, CurveIndex=$Emitter.Age, Scale Curve=-1)
    Index 1 Param 4=2
    Index 2 Param 1=30
    Index 2 Param 2=10
    Index 2 Param 3=0.6
    Index 2 Param 4=0.6
    Module.Index0Float0Write=true
    Module.Index0Float1Write=true
    Module.Index0Float2Write=true
    Module.Index0Float3Write=true
    Module.Index1Float0Write=true
    Module.Index1Float1Write=true
    Module.Index1Float2Write=true
    Module.Index1Float3Write=true
    Module.Index2Float0Write=true
    Module.Index2Float1Write=true
    Module.Index2Float2Write=false
    Module.Index2Float3Write=false
    Write Parameter Index 1=::NewEnumerator1
    Write Parameter Index 2=::NewEnumerator1
   )
  ]
  rendererClasses [&NiagaraRibbonRendererProperties]
  renderers [{rendererClass=&NiagaraRibbonRendererProperties}]
 }
}
summary {
 systemName=NS_Pickaxe_HybridBoot_Trail
 emitters [
  {emitterName=MainTrail, rendererClasses [&NiagaraRibbonRendererProperties]}
  {emitterName=TrailTip, rendererClasses [&NiagaraRibbonRendererProperties]}
 ]
 userVariables [
  User.bIsFrontEnd {type=NiagaraBool}
  User.bIsFrontEndPreview {type=NiagaraBool}
  User.PhysicalSurface {type=NiagaraInt32}
  User.RibbonIDOffset {type=NiagaraInt32}
  User.SwingSectionIndex {type=NiagaraInt32}
 ]
}
